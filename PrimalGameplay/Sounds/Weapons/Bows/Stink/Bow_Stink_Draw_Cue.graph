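# Sound cue for drawing the Stink Bow, layering a draw one-shot with a sustained tension loop.
# A mixer combines one of three random draw sounds with a quieter looping tension sound borrowed from the Bone bow.
# The loop starts after a short delay, and an enveloper raises its volume and pitch over several seconds.
# It uses weapon foley class, reload attenuation and global and local bow concurrency rules.
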SoundCue /PrimalGameplay/Sounds/Weapons/Bows/Stink/Bow_Stink_Draw_Cue
root=Mixer_0
nodes [
 {id=Delay_0, children [WavePlayer_4], properties {delayMax=0.5, delayMin=0.5}}
 {
  id=Enveloper_0
  children [Delay_0]
  properties {
   pitchCurve {editorCurveData {keys [<at 0 0.9>,<at 10 1.2>]}}
   volumeCurve {editorCurveData {keys [
      {arriveTangent=-0.0479, interpMode=RCIM_Cubic, leaveTangent=-0.0479, tangentMode=RCTM_User, time=0.0299, value=0.0564}
      {arriveTangent=-0.0494, interpMode=RCIM_Cubic, leaveTangent=-0.0494, tangentMode=RCTM_User, time=5, value=1}
     ]}}
  }
 }
 {id=Mixer_0, children [Random_0,Modulator_2], properties {inputVolume [1,0.6]}}
 {id=Modulator_2, children [Enveloper_0], properties {volumeMax=0.7, volumeMin=0.7}}
 {id=Random_0, children [WavePlayer_1,WavePlayer_2,WavePlayer_3], properties {weights [1,1,1]}}
 {id=WavePlayer_1, properties {soundWaveAssetPtr=&/Bow_Stink_Draw_01}}
 {id=WavePlayer_2, properties {soundWaveAssetPtr=&/Bow_Stink_Draw_02}}
 {id=WavePlayer_3, properties {soundWaveAssetPtr=&/Bow_Stink_Draw_03}}
 {id=WavePlayer_4, properties {bLooping=true, soundWaveAssetPtr=/PrimalGameplay/Sounds/Weapons/Bows/Bone/Bow_Bone_Draw_TensionLoop_01}}
]
properties {
 attenuationSettings=&Fort_Weapon_Reload_Att
 bDebug=true
 maxDistance=2250
 soundClassObject=&Fort_WeaponFoley
 volumeMultiplier=0.501
 concurrencySet [/Game/Sounds/Concurrency_Rules/Bow_Global_Conc, /Game/Sounds/Concurrency_Rules/Bow_Local_Conc]
}
