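# First-person Sound Cue for firing the Stink bow at full tension.
# A mixer layers one of three reused Metal bow shot waves with one of three Stink ooze waves, each picked at random and passed through a modulator.
# The ooze layer starts after a short fixed delay and gets a random upward pitch shift.
# It uses first-person ranged-fire sound class, projectile attenuation and gunfire concurrency settings.
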SoundCue /PrimalGameplay/Sounds/Weapons/Bows/Stink/Bow_Stink_Fire_Full_Tension_1P_Cue
root=Mixer_1
nodes [
 {id=Delay_0, children [Modulator_2], properties {delayMax=0.12, delayMin=0.12}}
 {id=Mixer_1, children [Modulator_0,Delay_0], properties {inputVolume [0.6,0.6]}}
 {id=Modulator_0, children [Random_1]}
 {id=Modulator_2, children [Random_3], properties {pitchMax=1.25, pitchMin=1}}
 {id=Random_1, children [WavePlayer_0,WavePlayer_1,WavePlayer_4], properties {weights [1,1,1]}}
 {id=Random_3, children [WavePlayer_6,WavePlayer_7,WavePlayer_8], properties {weights [1,1,1]}}
 {id=WavePlayer_0, properties {soundWaveAssetPtr=/PrimalGameplay/Sounds/Weapons/Bows/Metal/Bow_Metal_Fire_Full_Tension_01}}
 {id=WavePlayer_1, properties {soundWaveAssetPtr=/PrimalGameplay/Sounds/Weapons/Bows/Metal/Bow_Metal_Fire_Full_Tension_02}}
 {id=WavePlayer_4, properties {soundWaveAssetPtr=/PrimalGameplay/Sounds/Weapons/Bows/Metal/Bow_Metal_Fire_Full_Tension_03}}
 {id=WavePlayer_6, properties {soundWaveAssetPtr=&/Bow_Stink_Fire_Full_Tension_Close_Ooze_03}}
 {id=WavePlayer_7, properties {soundWaveAssetPtr=&/Bow_Stink_Fire_Full_Tension_Close_Ooze_01}}
 {id=WavePlayer_8, properties {soundWaveAssetPtr=&/Bow_Stink_Fire_Full_Tension_Close_Ooze_02}}
]
properties {
 attenuationSettings=&Fort_Weapon_Projectile_FirstPerson_Athena_Att
 maxDistance=2097152
 soundClassObject=&Fort_WeaponRanged_Fire_1P_PSM
 volumeMultiplier=1
 attenuationOverrides {falloffDistance=5000, attenuationShapeExtents {x=1000}}
 concurrencySet [&Weapon_Gunfire_FirstPerson_Conc]
}
