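# Sound cue for firing the Stink bow at full tension, built as close and distant layers with separate attenuation and sound classes.
# The close third-person layer mixes random low-tension fire shots with a quieter, slightly delayed ooze layer that gets random upward pitch variation.
# The distant layer plays a random distant shot variant.
# Global and local bow concurrency rules limit simultaneous playback.
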SoundCue /PrimalGameplay/Sounds/Weapons/Bows/Stink/Bow_Stink_Fire_Full_Tension_Cue
root=Mixer_3
nodes [
 {id=Attenuation_0, children [Mixer_2], properties {attenuationSettings=/Game/Sounds/Attenuation/Fort_Weapon_Projectile_Close_Bow_Athena_Att}}
 {
  id=Attenuation_2
  children [Modulator_8]
  properties {attenuationSettings=/Game/Sounds/Attenuation/Fort_Weapon_Projectile_Distant_Bow_Athena_Att}
 }
 {id=Delay_0, children [Modulator_2], properties {delayMax=0.075, delayMin=0.075}}
 {id=Mixer_2, children [Modulator_3,Delay_0], properties {inputVolume [1,0.35]}}
 {id=Mixer_3, children [SoundClass_0,SoundClass_4], properties {inputVolume [1,1]}}
 {id=Modulator_2, children [Random_1], properties {pitchMax=1.25, pitchMin=1}}
 {id=Modulator_3, children [Random_0]}
 {id=Modulator_8, children [Random_6]}
 {id=Random_0, children [WavePlayer_0,WavePlayer_1,WavePlayer_7], properties {weights [1,1,1]}}
 {id=Random_1, children [WavePlayer_6,WavePlayer_2,WavePlayer_3], properties {weights [1,1,1]}}
 {id=Random_6, children [WavePlayer_21, WavePlayer_22, WavePlayer_23], properties {weights [1,1,1]}}
 {
  id=SoundClass_0
  children [Attenuation_0]
  properties {soundClassOverride=&Fort_WeaponRanged_Fire_3P_PSM}
 }
 {
  id=SoundClass_4
  children [Attenuation_2]
  properties {soundClassOverride=&Fort_WeaponRanged_Fire_Distant}
 }
 {id=WavePlayer_0, properties {soundWaveAssetPtr=&/Bow_Stink_Fire_Low_Tension_Close_03}}
 {id=WavePlayer_1, properties {soundWaveAssetPtr=&/Bow_Stink_Fire_Low_Tension_Close_02}}
 {id=WavePlayer_2, properties {soundWaveAssetPtr=&/Bow_Stink_Fire_Full_Tension_Close_Ooze_01}}
 {id=WavePlayer_21, properties {soundWaveAssetPtr=&/Bow_Stink_Fire_3P_Distant_01}}
 {id=WavePlayer_22, properties {soundWaveAssetPtr=&/Bow_Stink_Fire_3P_Distant_02}}
 {id=WavePlayer_23, properties {soundWaveAssetPtr=&/Bow_Stink_Fire_3P_Distant_03}}
 {id=WavePlayer_3, properties {soundWaveAssetPtr=&/Bow_Stink_Fire_Full_Tension_Close_Ooze_02}}
 {id=WavePlayer_6, properties {soundWaveAssetPtr=&/Bow_Stink_Fire_Full_Tension_Close_Ooze_03}}
 {id=WavePlayer_7, properties {soundWaveAssetPtr=&/Bow_Stink_Fire_Low_Tension_Close_01}}
]
properties {
 maxDistance=27500
 volumeMultiplier=1
 attenuationOverrides {falloffDistance=5000, attenuationShapeExtents {x=1000}}
 concurrencySet [/Game/Sounds/Concurrency_Rules/Bow_Global_Conc, /Game/Sounds/Concurrency_Rules/Bow_Local_Conc]
}
