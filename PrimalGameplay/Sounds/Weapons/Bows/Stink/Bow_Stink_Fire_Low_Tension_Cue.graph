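# Sound cue for firing the Stink bow at low tension, built entirely from wave assets in the Metal bow folder.
# A root mixer blends a close third-person layer and a distant layer, each with its own sound class and projectile attenuation.
# The close layer mixes a random low-tension fire shot with a briefly delayed random passby, both through modulators.
# The distant layer picks a random distant fire wave, and global and local bow concurrency rules govern the cue.
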SoundCue /PrimalGameplay/Sounds/Weapons/Bows/Stink/Bow_Stink_Fire_Low_Tension_Cue
root=Mixer_0
nodes [
 {id=Attenuation_1, children [Mixer_2], properties {attenuationSettings=/Game/Sounds/Attenuation/Fort_Weapon_Projectile_Close_Bow_Athena_Att}}
 {
  id=Attenuation_3
  children [Random_0]
  properties {attenuationSettings=/Game/Sounds/Attenuation/Fort_Weapon_Projectile_Distant_Bow_Athena_Att}
 }
 {id=Delay_0, children [Modulator_4], properties {delayMax=0.18, delayMin=0.18}}
 {id=Mixer_0, children [SoundClass_1,SoundClass_2], properties {inputVolume [0.65,0.7]}}
 {id=Mixer_2, children [Modulator_2,Delay_0], properties {inputVolume [1,1]}}
 {id=Modulator_2, children [Random_3]}
 {id=Modulator_4, children [Random_4]}
 {id=Random_0, children [WavePlayer_3,WavePlayer_4,WavePlayer_5], properties {weights [1,1,1]}}
 {id=Random_3, children [WavePlayer_10, WavePlayer_11, WavePlayer_12], properties {weights [1,1,1]}}
 {
  id=Random_4
  children [WavePlayer_13, WavePlayer_14, WavePlayer_15, WavePlayer_16, WavePlayer_17, WavePlayer_18]
  properties {weights [1,1,1,1,1,1]}
 }
 {
  id=SoundClass_1
  children [Attenuation_1]
  properties {soundClassOverride=&Fort_WeaponRanged_Fire_3P_PSM}
 }
 {
  id=SoundClass_2
  children [Attenuation_3]
  properties {soundClassOverride=&Fort_WeaponRanged_Fire_Distant}
 }
 {id=WavePlayer_10, properties {soundWaveAssetPtr=/PrimalGameplay/Sounds/Weapons/Bows/Metal/Bow_Metal_Fire_Low_Tension_01}}
 {id=WavePlayer_11, properties {soundWaveAssetPtr=/PrimalGameplay/Sounds/Weapons/Bows/Metal/Bow_Metal_Fire_Low_Tension_02}}
 {id=WavePlayer_12, properties {soundWaveAssetPtr=/PrimalGameplay/Sounds/Weapons/Bows/Metal/Bow_Metal_Fire_Low_Tension_03}}
 {id=WavePlayer_13, properties {soundWaveAssetPtr=/PrimalGameplay/Sounds/Weapons/Bows/Metal/Bow_Metal_Passby_Default_Distant_06}}
 {id=WavePlayer_14, properties {soundWaveAssetPtr=/PrimalGameplay/Sounds/Weapons/Bows/Metal/Bow_Metal_Passby_Default_Distant_01}}
 {id=WavePlayer_15, properties {soundWaveAssetPtr=/PrimalGameplay/Sounds/Weapons/Bows/Metal/Bow_Metal_Passby_Default_Distant_02}}
 {id=WavePlayer_16, properties {soundWaveAssetPtr=/PrimalGameplay/Sounds/Weapons/Bows/Metal/Bow_Metal_Passby_Default_Distant_03}}
 {id=WavePlayer_17, properties {soundWaveAssetPtr=/PrimalGameplay/Sounds/Weapons/Bows/Metal/Bow_Metal_Passby_Default_Distant_04}}
 {id=WavePlayer_18, properties {soundWaveAssetPtr=/PrimalGameplay/Sounds/Weapons/Bows/Metal/Bow_Metal_Passby_Default_Distant_05}}
 {id=WavePlayer_3, properties {soundWaveAssetPtr=/PrimalGameplay/Sounds/Weapons/Bows/Metal/Bow_Default_Fire_3P_Normal_Distant_01}}
 {id=WavePlayer_4, properties {soundWaveAssetPtr=/PrimalGameplay/Sounds/Weapons/Bows/Metal/Bow_Default_Fire_3P_Normal_Distant_02}}
 {id=WavePlayer_5, properties {soundWaveAssetPtr=/PrimalGameplay/Sounds/Weapons/Bows/Metal/Bow_Default_Fire_3P_Normal_Distant_03}}
]
properties {
 maxDistance=27500
 volumeMultiplier=1
 attenuationOverrides {falloffDistance=5000, attenuationShapeExtents {x=1000}}
 concurrencySet [/Game/Sounds/Concurrency_Rules/Bow_Global_Conc, /Game/Sounds/Concurrency_Rules/Bow_Local_Conc]
}
