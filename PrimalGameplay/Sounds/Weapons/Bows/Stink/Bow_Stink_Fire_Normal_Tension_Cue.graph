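# Sound cue for the Stink bow's normal-tension fire sound, mixing close and distant layers with their own attenuation and sound class.
# The close 3P layer plays a random Metal bow shot plus a quieter, slightly delayed random Stink ooze accent.
# The distant layer reuses random default-bow distant shots from the Metal folder.
# Modulators sit above each random pool, and global and local bow concurrency rules govern playback.
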SoundCue /PrimalGameplay/Sounds/Weapons/Bows/Stink/Bow_Stink_Fire_Normal_Tension_Cue
root=Mixer_0
nodes [
 {id=Attenuation_3, children [Mixer_4], properties {attenuationSettings=/Game/Sounds/Attenuation/Fort_Weapon_Projectile_Close_Bow_Athena_Att}}
 {
  id=Attenuation_4
  children [Modulator_1]
  properties {attenuationSettings=/Game/Sounds/Attenuation/Fort_Weapon_Projectile_Distant_Bow_Athena_Att}
 }
 {id=Delay_1, children [Modulator_2], properties {delayMax=0.1, delayMin=0.1}}
 {id=Mixer_0, children [SoundClass_2,SoundClass_3], properties {inputVolume [0.65,0.7]}}
 {id=Mixer_4, children [Modulator_6,Delay_1], properties {inputVolume [0.8,0.4]}}
 {id=Modulator_1, children [Random_2]}
 {id=Modulator_2, children [Random_0]}
 {id=Modulator_6, children [Random_5]}
 {id=Random_0, children [WavePlayer_1,WavePlayer_2,WavePlayer_3], properties {weights [1,1,1]}}
 {id=Random_2, children [WavePlayer_7, WavePlayer_9, WavePlayer_10], properties {weights [1,1,1]}}
 {id=Random_5, children [WavePlayer_19, WavePlayer_20, WavePlayer_24], properties {weights [1,1,1]}}
 {
  id=SoundClass_2
  children [Attenuation_3]
  properties {soundClassOverride=&Fort_WeaponRanged_Fire_3P_PSM}
 }
 {
  id=SoundClass_3
  children [Attenuation_4]
  properties {soundClassOverride=&Fort_WeaponRanged_Fire_Distant}
 }
 {id=WavePlayer_1, properties {soundWaveAssetPtr=&/Bow_Stink_Fire_Normal_Tension_Close_Ooze_01}}
 {id=WavePlayer_10, properties {soundWaveAssetPtr=/PrimalGameplay/Sounds/Weapons/Bows/Metal/Bow_Default_Fire_3P_Normal_Distant_02}}
 {id=WavePlayer_19, properties {soundWaveAssetPtr=/PrimalGameplay/Sounds/Weapons/Bows/Metal/Bow_Metal_Fire_Normal_Tension_01}}
 {id=WavePlayer_2, properties {soundWaveAssetPtr=&/Bow_Stink_Fire_Normal_Tension_Close_Ooze_02}}
 {id=WavePlayer_20, properties {soundWaveAssetPtr=/PrimalGameplay/Sounds/Weapons/Bows/Metal/Bow_Metal_Fire_Normal_Tension_02}}
 {id=WavePlayer_24, properties {soundWaveAssetPtr=/PrimalGameplay/Sounds/Weapons/Bows/Metal/Bow_Metal_Fire_Normal_Tension_03}}
 {id=WavePlayer_3, properties {soundWaveAssetPtr=&/Bow_Stink_Fire_Normal_Tension_Close_Ooze_03}}
 {id=WavePlayer_7, properties {soundWaveAssetPtr=/PrimalGameplay/Sounds/Weapons/Bows/Metal/Bow_Default_Fire_3P_Normal_Distant_03}}
 {id=WavePlayer_9, properties {soundWaveAssetPtr=/PrimalGameplay/Sounds/Weapons/Bows/Metal/Bow_Default_Fire_3P_Normal_Distant_01}}
]
properties {
 maxDistance=27500
 volumeMultiplier=1
 attenuationOverrides {falloffDistance=5000, attenuationShapeExtents {x=1000}}
 concurrencySet [/Game/Sounds/Concurrency_Rules/Bow_Global_Conc, /Game/Sounds/Concurrency_Rules/Bow_Local_Conc]
}
