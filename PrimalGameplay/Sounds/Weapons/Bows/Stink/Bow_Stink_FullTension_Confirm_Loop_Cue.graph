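# Sound cue for the Stink bow's full-tension confirm, layering a charge start hit with a sustained loop and recurring hisses.
# A mixer blends a random charged-start one-shot, a looping bed and a quieter hiss layer.
# An Enveloper fades the bed in over two seconds while raising its pitch over five.
# After a roughly two-second delay, a Looping node repeats random hiss picks through pitch and volume modulation.
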SoundCue /PrimalGameplay/Sounds/Weapons/Bows/Stink/Bow_Stink_FullTension_Confirm_Loop_Cue
root=Mixer_0
nodes [
 {id=Delay_0, children [Modulator_0], properties {delayMax=2.2, delayMin=1.8}}
 {
  id=Enveloper_0
  children [WavePlayer_1]
  properties {
   pitchCurve {editorCurveData {keys [<at 0 0.8>,<at 5 1.2>]}}
   volumeCurve {editorCurveData {keys [
      <at 0 0 RCIM_Cubic>
      {arriveTangent=0.8706, interpMode=RCIM_Cubic, leaveTangent=0.8706, tangentMode=RCTM_User, time=1.3491, value=0.3189}
      {arriveTangent=-0.0615, interpMode=RCIM_Cubic, leaveTangent=-0.0615, tangentMode=RCTM_User, time=2, value=1}
     ]}}
  }
 }
 {id=Looping_0, children [Random_1]}
 {id=Mixer_0, children [Delay_0,Enveloper_0,Modulator_1], properties {inputVolume [0.15,0.3,0.9]}}
 {id=Modulator_0, children [Looping_0], properties {pitchMax=1.25, pitchMin=0.9, volumeMin=0.6}}
 {id=Modulator_1, children [Random_0], properties {pitchMin=1}}
 {id=Random_0, children [WavePlayer_3,WavePlayer_4,WavePlayer_5,WavePlayer_6], properties {weights [1,1,1,1]}}
 {
  id=Random_1
  children [
   WavePlayer_7
   WavePlayer_8
   WavePlayer_9
   WavePlayer_10
   WavePlayer_11
   WavePlayer_12
   WavePlayer_13
   WavePlayer_14
   WavePlayer_15
   WavePlayer_16
  ]
  properties {weights [1,1,1,1,1,1,1,1,1,1]}
 }
 {id=WavePlayer_1, properties {bLooping=true, soundWaveAssetPtr=&/Bow_Stink_Full_Tension_Confirm_Loop}}
 {id=WavePlayer_10, properties {soundWaveAssetPtr=&/Bow_Stink_Full_Tension_Confirm_Hiss_03}}
 {id=WavePlayer_11, properties {soundWaveAssetPtr=&/Bow_Stink_Full_Tension_Confirm_Hiss_04}}
 {id=WavePlayer_12, properties {soundWaveAssetPtr=&/Bow_Stink_Full_Tension_Confirm_Hiss_05}}
 {id=WavePlayer_13, properties {soundWaveAssetPtr=&/Bow_Stink_Full_Tension_Confirm_Hiss_06}}
 {id=WavePlayer_14, properties {soundWaveAssetPtr=&/Bow_Stink_Full_Tension_Confirm_Hiss_07}}
 {id=WavePlayer_15, properties {soundWaveAssetPtr=&/Bow_Stink_Full_Tension_Confirm_Hiss_08}}
 {id=WavePlayer_16, properties {soundWaveAssetPtr=&/Bow_Stink_Full_Tension_Confirm_Hiss_09}}
 {id=WavePlayer_3, properties {soundWaveAssetPtr=&/Bow_Stink_Prj_Full_Tension_Charged_Start_04}}
 {id=WavePlayer_4, properties {soundWaveAssetPtr=&/Bow_Stink_Prj_Full_Tension_Charged_Start_01}}
 {id=WavePlayer_5, properties {soundWaveAssetPtr=&/Bow_Stink_Prj_Full_Tension_Charged_Start_02}}
 {id=WavePlayer_6, properties {soundWaveAssetPtr=&/Bow_Stink_Prj_Full_Tension_Charged_Start_03}}
 {id=WavePlayer_7, properties {soundWaveAssetPtr=&/Bow_Stink_Full_Tension_Confirm_Hiss_10}}
 {id=WavePlayer_8, properties {soundWaveAssetPtr=&/Bow_Stink_Full_Tension_Confirm_Hiss_01}}
 {id=WavePlayer_9, properties {soundWaveAssetPtr=&/Bow_Stink_Full_Tension_Confirm_Hiss_02}}
]
properties {
 attenuationSettings=/Game/Sounds/Attenuation/Fort_Weapon_Projectile_Bow_Charge_Confirm_Loop_Att
 bDebug=true
 maxDistance=2100
 soundClassObject=&Fort_WeaponFoley
 volumeMultiplier=0.501
 concurrencySet [/Game/Sounds/Concurrency_Rules/Bow_Global_Conc, /Game/Sounds/Concurrency_Rules/Bow_Local_Conc]
}
