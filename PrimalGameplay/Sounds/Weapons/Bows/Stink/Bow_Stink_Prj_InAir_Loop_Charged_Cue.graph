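# Sound cue for the in-flight sound of a charged Stink Bow projectile.
# A mixer layers one of six randomly picked fire whooshes with a looping in-air wave.
# An enveloper on the loop ramps volume up while pitch drops from a raised start.
# It uses the arrow in-air attenuation asset and shares bow global and local concurrency rules.
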SoundCue /PrimalGameplay/Sounds/Weapons/Bows/Stink/Bow_Stink_Prj_InAir_Loop_Charged_Cue
root=Mixer_1
nodes [
 {
  id=Enveloper_0
  children [WavePlayer_1]
  properties {pitchCurve {editorCurveData {keys [<at 0 2>, {time=0.5}]}}, volumeCurve {editorCurveData {keys [<at 0 0.15 RCIM_Cubic>,<at 1 1>]}}}
 }
 {id=Mixer_1, children [Random_0,Enveloper_0], properties {inputVolume [1,1]}}
 {id=Random_0, children [WavePlayer_0,WavePlayer_2,WavePlayer_3,WavePlayer_4,WavePlayer_5,WavePlayer_6], properties {weights [1,1,1,1,1,1]}}
 {id=WavePlayer_0, properties {soundWaveAssetPtr=&/Bow_Stink_Fire_Whoosh_01}}
 {id=WavePlayer_1, properties {bLooping=true, soundWaveAssetPtr=&/Bow_Stink_Prj_InAir_Loop}}
 {id=WavePlayer_2, properties {soundWaveAssetPtr=&/Bow_Stink_Fire_Whoosh_02}}
 {id=WavePlayer_3, properties {soundWaveAssetPtr=&/Bow_Stink_Fire_Whoosh_03}}
 {id=WavePlayer_4, properties {soundWaveAssetPtr=&/Bow_Stink_Fire_Whoosh_04}}
 {id=WavePlayer_5, properties {soundWaveAssetPtr=&/Bow_Stink_Fire_Whoosh_05}}
 {id=WavePlayer_6, properties {soundWaveAssetPtr=&/Bow_Stink_Fire_Whoosh_06}}
]
properties {
 attenuationSettings=/Game/Sounds/Attenuation/Fort_Weapon_Projectile_Arrow_InAirLoop_Att
 maxDistance=16800
 soundClassObject=&Fort_WeaponRanged_Projectile_Effects
 volumeMultiplier=1
 attenuationOverrides {falloffDistance=5000, reverbWetLevelMax=0.7, reverbWetLevelMin=0.2, attenuationShapeExtents {x=1000}}
 concurrencySet [/Game/Sounds/Concurrency_Rules/Bow_Global_Conc, /Game/Sounds/Concurrency_Rules/Bow_Local_Conc]
}
