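# Sound cue for the reload of the Stink bow, using weapon foley class, reload attenuation and bow concurrency rules.
# A mixer layers a random pick from three reload waves with a random pick from four knock waves, the latter starting after a short fixed delay and mixed slightly louder.
# Both layers get random pitch variation, and the knock layer also varies in volume.
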SoundCue /PrimalGameplay/Sounds/Weapons/Bows/Stink/Bow_Stink_Reload_Cue
root=Mixer_0
nodes [
 {id=Delay_0, children [Modulator_1], properties {delayMax=0.34, delayMin=0.34}}
 {id=Mixer_0, children [Modulator_0,Delay_0], properties {inputVolume [1,1.1]}}
 {id=Modulator_0, children [Random_1], properties {pitchMax=1.2, pitchMin=0.9}}
 {id=Modulator_1, children [Random_2], properties {pitchMax=1.2, pitchMin=0.9, volumeMin=0.8}}
 {id=Random_1, children [WavePlayer_1,WavePlayer_5,WavePlayer_6], properties {weights [1,1,1]}}
 {id=Random_2, children [WavePlayer_8, WavePlayer_9, WavePlayer_10, WavePlayer_11], properties {weights [1,1,1,1]}}
 {id=WavePlayer_1, properties {soundWaveAssetPtr=&/Bow_Stink_Reload_01}}
 {id=WavePlayer_10, properties {soundWaveAssetPtr=&/Bow_Stink_Reload_Knock_03}}
 {id=WavePlayer_11, properties {soundWaveAssetPtr=&/Bow_Stink_Reload_Knock_04}}
 {id=WavePlayer_5, properties {soundWaveAssetPtr=&/Bow_Stink_Reload_02}}
 {id=WavePlayer_6, properties {soundWaveAssetPtr=&/Bow_Stink_Reload_03}}
 {id=WavePlayer_8, properties {soundWaveAssetPtr=&/Bow_Stink_Reload_Knock_01}}
 {id=WavePlayer_9, properties {soundWaveAssetPtr=&/Bow_Stink_Reload_Knock_02}}
]
properties {
 attenuationSettings=&Fort_Weapon_Reload_Att
 bDebug=true
 maxDistance=2250
 soundClassObject=&Fort_WeaponFoley
 volumeMultiplier=0.2
 concurrencySet [/Game/Sounds/Concurrency_Rules/Bow_Global_Conc, /Game/Sounds/Concurrency_Rules/Bow_Local_Conc]
}
